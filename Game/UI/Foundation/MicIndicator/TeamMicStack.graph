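# HUD element that stacks a Cinematic-TeamMics widget for each teammate in a bordered vertical box.
# Construct clears placeholders, binds party and team events, and feeds existing members after the first into HandleTeamMemberChanged.
# That handler skips the local player and updates or appends entries by PlayerIndex minus one; removal uses the same offset.
# The border hides when empty; ClearContents clears children but leaves the TeamMembers array intact.
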WidgetBlueprint /Game/UI/Foundation/MicIndicator/TeamMicStack
graphs [
 AppendTeamMember {nodes [
   n1 = AppendTeamMember!() -> (then, TeamMemberInfo:"Fort Team Member Info Structure (by ref)") {is=title K2Node_FunctionEntry}
   > n2 = Create Widget!(
    execute:Exec
    Class:User Widget Class Reference=&/Cinematic-TeamMics.Cinematic-TeamMics_C
    OwningPlayer:Player Controller Object Reference=GetOwningPlayer()
    MemberInfo:Fort Team Member Info Structure=@n1
   ) -> (then:Exec, ReturnValue:Cinematic- Team Mics Object Reference)
   > Add(
    TargetArray:Array of Cinematic- Team Mics Object References=Get TeamMembers() -> (TeamMembers:Array of Cinematic- Team Mics Object References)
    NewItem:"Cinematic- Team Mics Object Reference (by ref)"=@n2
   ) {is=title K2Node_CallArrayFunction}
   > n3 = AddChildToVerticalBox(self="Get [Vertical Box] Team Members"() -> ("[Vertical Box] Team Members":Vertical Box Object Reference), Content=@n2)
   > SetPadding!(execute, self:Vertical Box Slot Object Reference=@n3, InPadding:Margin Structure=Make Margin(Left=0, Top=8, Right=0, Bottom=0)) -> (then)
   > SetVisibility(self=Get BorderFrame() -> (BorderFrame:Common Border Object Reference), InVisibility=SelfHitTestInvisible)
  ]}
 ClearContents {nodes [
   ClearContents() {is=title K2Node_FunctionEntry}
   > ClearChildren(self="Get [Vertical Box] Team Members"())
   > SetVisibility(self=Get BorderFrame(), InVisibility=Collapsed)
  ]}
 EventGraph {nodes [
   HandlePartyLeft() {is=title K2Node_CustomEvent}
   > ClearContents() {is=title K2Node_CallFunction}
   Event Construct()
   > ClearContents()
   n1 = Get FortPartyContext!() -> (ReturnValue:Fort Party Context Object Reference) {is=title K2Node_GetBlueprintContextEx}
   > Bind Event to On Party Left(self:Fort Party Context Object Reference=@n1, Delegate=Create Event()) {is=title K2Node_AddDelegate}
   n2 = GetLocalPlayerTeam!(self:Fort Party Context Object Reference=@n1) -> (ReturnValue:Fort UITeam Info Object Reference)
   > Bind Event to On Team Member Added(self:Fort UITeam Info Object Reference=@n2, Delegate=Create Event()) {is=title K2Node_AddDelegate}
   > Bind Event to On Team Member Removed(self:Fort UITeam Info Object Reference=@n2, Delegate=Create Event()) {is=title K2Node_AddDelegate}
   > Bind Event to On Team Member State Changed(self:Fort UITeam Info Object Reference=@n2, Delegate=Create Event()) {is=title K2Node_AddDelegate}
   > n3 = GetTeamMembers!(execute, self:Fort UITeam Info Object Reference=@n2) -> (then, TeamMembers:Array of Fort Team Member Info Structures)
   > n4 = For Each Loop(Array:Array of Fort Team Member Info Structures=@n3) -> (Array Element:Fort Team Member Info Structure)
   > Branch(Condition="integer > integer"(A=@n4.Array Index, B=0) {is=title K2Node_CallFunction})
   > HandleTeamMemberChanged!(execute, self, Member Info:Fort Team Member Info Structure=@n4) -> (then)
  ]}
 HandleTeamMemberChanged {nodes [
   n1 = HandleTeamMemberChanged!() -> (then, Member Info:Fort Team Member Info Structure) {is=title K2Node_FunctionEntry}
   n2 = Break Fort Team Member Info!(FortTeamMemberInfo:"Fort Team Member Info Structure (by ref)"=@n1) -> (
    MemberUniqueId:Unique Net Id Structure
    PartyLeaderUniqueId:Unique Net Id Structure
    ConsoleUniqueId:Unique Net Id Structure
    Platform:String
    PlayerName:Text
    bPartyLeader:bool=~
    bIsInZone:bool=~
    bHasBoostXp:bool=~
    bHasRestXp:bool=~
    BannerIconId:String
    BannerColorId:String
    bBattlePassPurchased:bool=~
    BattlePassLevel:int
    NumPlayersInParty:int
    PlayerIndex:int
    TeamAffiliation:Byte
    HeroClass:Text
    HeroLevel:Text
    HeroXP:int
    HeroItem:Fort Item Object Reference
    SelectedGadgetItems:Array of Fort Item Object References
    HeroIcon:Slate Brush Structure
    HomeBaseInfo:Fort Home Base Info Structure
   ) {is=title K2Node_BreakStruct}
   > n3 = Branch(Condition=IsLocalPlayersUniqueID!(
     self:Fort Mcp Context Object Reference=Get FortMcpContext!() -> (ReturnValue:Fort Mcp Context Object Reference) {is=title K2Node_GetBlueprintContextEx}
     PlayerID:"Unique Net Id Structure (by ref)"=@n2
    ) -> (ReturnValue:bool))
   n4 = Get TeamMembers()
   n5 = int - int(A=@n2.PlayerIndex, B=1) {is=title K2Node_CallFunction}
   n6 = Branch(
    execute=@n3.else
    Condition="integer < integer"!(A:int=@n5, B:int=Length(TargetArray:Array of Cinematic- Team Mics Object References=@n4)) -> (ReturnValue:bool) {is=title K2Node_CallFunction}
   )
   > Update!(
    execute
    self:Cinematic- Team Mics Object Reference="Get (a copy)"(Array:Array of Cinematic- Team Mics Object References=@n4, Dimension 1=@n5) -> (Output:Cinematic- Team Mics Object Reference)
    UpdatedMemberInfo:Fort Team Member Info Structure=@n1
   ) -> (then)
   AppendTeamMember!(execute=@n6.else, self, TeamMemberInfo:"Fort Team Member Info Structure (by ref)"=@n1) -> (then) {is=title K2Node_CallFunction}
  ]}
 HandleTeamMemberRemoved {nodes [
   n1 = HandleTeamMemberRemoved!() -> (then, Member Index:int) {is=title K2Node_FunctionEntry}
   n2 = int - int(A=@n1, B=1)
   > Remove Index!(execute, TargetArray:Array of Cinematic- Team Mics Object References=Get TeamMembers(), IndexToRemove:int=@n2) -> (then)
   n3 = "Get [Vertical Box] Team Members"()
   > RemoveChildAt!(execute, self:Panel Widget Object Reference=@n3, Index:int=@n2) -> (then, ReturnValue:bool)
   > Branch(Condition="Equal (Integer)"(A=GetChildrenCount!(self:Panel Widget Object Reference=@n3) -> (ReturnValue:int=0), B=0) {is=title K2Node_CallFunction})
   > SetVisibility(self=Get BorderFrame(), InVisibility=Collapsed)
  ]}
]
variables [TeamMembers {type="array<Cinematic-TeamMics_C>", default []}]
widgets {
 info {inheritedWidgetCount=1, parentClass=&FortHUDElementWidget, rootWidgetClass=&CommonBorder, widgetCount=7}
 widgets [
  BorderFrame {widgetClassPath=&CommonBorder, bIsVariable=true, widget=BorderFrame, properties {padding=<ltrb 4 4 4 4>}}
  "[Vertical Box] Team Members" {
   widgetClassPath=&VerticalBox
   bIsVariable=true
   parent=BorderFrame
   widget="[Vertical Box] Team Members"
   slot {class=BorderSlot, properties {padding=<ltrb 4 4 4 4>}}
  }
  Cinematic-TeamMics {
   widgetClassPath=&/Cinematic-TeamMics
   parent="[Vertical Box] Team Members"
   widget=Cinematic-TeamMics
   slot {class=VerticalBoxSlot}
  }
  Cinematic-TeamMics_C_0 {
   widgetClassPath=&Cinematic-TeamMics
   parent="[Vertical Box] Team Members"
   widget=Cinematic-TeamMics_C_0
   slot {class=VerticalBoxSlot}
  }
  Cinematic-TeamMics_C_1 {
   widgetClassPath=&Cinematic-TeamMics
   parent="[Vertical Box] Team Members"
   widget=Cinematic-TeamMics_C_1
   slot {class=VerticalBoxSlot}
  }
  Cinematic-TeamMics_C_2 {
   widgetClassPath=&Cinematic-TeamMics
   parent="[Vertical Box] Team Members"
   widget=Cinematic-TeamMics_C_2
   slot {class=VerticalBoxSlot}
  }
  TutorialHighlight {widgetClassPath=&TutorialHighlight}
 ]
}
