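# Legacy movie widget, derived from FortMovieWidget, that shows video over a black background in a scale box, optionally keeping aspect ratio.
# Init sets the media source, playing immediately if a player already exists, otherwise waiting for the media-opened event to autoplay or broadcast readiness.
# The media texture feeds the image's dynamic material, and dimension updates resize its brush.
# Failed opens fire a dispatcher; destruction stops playback.
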WidgetBlueprint /Game/UI/Foundation/Movie/Legacy_MovieWidget
graphs [
 EventGraph {nodes [
   n1 = OnMediaLoaded() {is=title K2Node_CustomEvent}
   > Unbind Event from On Media Opened(self:Legacy Movie Widget Object Reference, Delegate=@n1) {is=title K2Node_RemoveDelegate}
   > SetDelayedContentValues()
   > n2 = Branch(Condition=Get AutoPlay() -> (AutoPlay:bool))
   > Play!(execute, self, ShouldRewind:bool=true) -> (then)
   Call On Media Ready to Play!(execute:Exec=@n2.else, self:Legacy Movie Widget Object Reference) -> (then:Exec) {is=title K2Node_CallDelegate}
   Event Destruct()
   > StopPlaying()
   n3 = Event MediaDimensionsUpdated!() -> (OutputDelegate, then, NewWidth:float, NewHeight:float) {is=title K2Node_Event}
   n4 = Break Slate Brush!(SlateBrush:"Slate Brush Structure (by ref)"=Get Brush!(self:Image Object Reference=Get Video_Image() -> (Video_Image:Image Object Reference)) -> (Brush:Slate Brush Structure)) -> (
    TintColor:Slate Color Structure
    DrawAs:ESlateBrushDrawType Enum
    Tiling:ESlateBrushTileType Enum
    Mirroring:ESlateBrushMirrorType Enum
    ImageSize:Vector 2D Structure
    Margin:Margin Structure
    ResourceObject:Object Reference
    OutlineSettings:Slate Brush Outline Settings Structure
   ) {is=title K2Node_BreakStruct}
   > SetBrush(
    self=Get Video_Image()
    InBrush=Make Slate Brush!(
     TintColor:Slate Color Structure=@n4
     DrawAs:ESlateBrushDrawType Enum=@n4.DrawAs
     Tiling:ESlateBrushTileType Enum=@n4.Tiling
     Mirroring:ESlateBrushMirrorType Enum=@n4.Mirroring
     ImageSize:Vector 2D Structure=MakeVector2D(X=@n3.NewWidth, Y=@n3.NewHeight)
     Margin:Margin Structure=@n4.Margin
     ResourceObject:Object Reference=@n4.ResourceObject
     OutlineSettings:Slate Brush Outline Settings Structure=(CornerRadii=(X=0,Y=0,Z=0,W=0),Color=(SpecifiedColor=(R=0,G=0,B=0,A=0),ColorUseRule=UseColor_Specified),Width=0,RoundingType=HalfHeightRadius,bUseBrushTransparency=False)
    ) -> (SlateBrush:Slate Brush Structure) {is=title K2Node_MakeStruct}
   )
  ]}
 Init {nodes [
   n1 = Init!() -> (
    then
    InMediaPlayer:Media Player Object Reference
    InMediaSource:Media Source Object Reference
    InMediaSoundWave:Object Reference
    AudioTrack:int
   ) {is=title K2Node_FunctionEntry}
   > n2 = Branch(Condition=HasPlayerForSource!(self:Fort Movie Widget Object Reference, InMediaSource:Media Source Object Reference=@n1.InMediaSource) -> (ReturnValue:bool))
   n3 = Make Media Player Options!(
    Tracks:Media Player Track Options Structure=Make Media Player Track Options!(Audio:int=@n1.AudioTrack, Caption:int=-1, Metadata:int=-1, Script:int=-1, Subtitle:int=-1, Text:int=-1, Video:int=0) -> (MediaPlayerTrackOptions:Media Player Track Options Structure) {is=title K2Node_MakeStruct}
    TracksByLanguage:Media Player Initial Track Language Selection Structure=(Video="",Audio="",Subtitle="",Caption="")
    TrackSelection:EMediaPlayerOptionTrackSelectMode Enum=UseTrackOptionIndices
    SeekTime:Timespan Structure=+00000000.00:00:0
    SeekTimeType:EMediaPlayerOptionSeekTimeType Enum=RelativeToStartTime
    PlayOnOpen:EMediaPlayerOptionBooleanOverride Enum=UseMediaPlayerSetting
    Loop:EMediaPlayerOptionBooleanOverride Enum=UseMediaPlayerSetting
   ) -> (MediaPlayerOptions:Media Player Options Structure) {is=title K2Node_MakeStruct}
   > n4 = SetMediaSource!(
    execute
    self:Fort Movie Widget Object Reference
    InMediaPlayer:Media Player Object Reference=@n1
    InMediaSource:Media Source Object Reference=@n1.InMediaSource
    PlayerOptions:"Media Player Options Structure (by ref)"=@n3
    bPlayOnOpen:bool=Get AutoPlay()
    bLooping:bool=Get ShouldLoop() -> (ShouldLoop:bool)
   ) -> (then, ReturnValue:bool)
   > n5 = Branch(Condition=@n4)
   > SetDelayedContentValues()
   > RequestPlayMovie!(execute, self:Fort Movie Widget Object Reference, bShouldRewind:bool=true) -> (then)
   Call On Media Failed to Open!(execute:Exec=[@n7.else, @n5.else], self:Legacy Movie Widget Object Reference) -> (then:Exec) {is=title K2Node_CallDelegate}
   Bind Event to On Media Opened(execute=@n2.else, self:Legacy Movie Widget Object Reference, Delegate=Create Event()) {is=title K2Node_AddDelegate}
   > n6 = SetMediaSource(InMediaPlayer=@n1, InMediaSource=@n1.InMediaSource, PlayerOptions=@n3, bPlayOnOpen=Get AutoPlay(), bLooping=Get ShouldLoop())
   > n7 = Branch(Condition=@n6)
  ]}
 OnMediaFailedToOpen {nodes [OnMediaFailedToOpen() {is=title K2Node_FunctionEntry}]}
 OnMediaReadyToPlay {nodes [OnMediaReadyToPlay() {is=title K2Node_FunctionEntry}]}
 Play {nodes [
   n1 = Play!() -> (then, ShouldRewind:bool=true) {is=title K2Node_FunctionEntry}
   > RequestPlayMovie(bShouldRewind=@n1)
  ]}
 SetDelayedContentValues {nodes [
   SetDelayedContentValues() {is=title K2Node_FunctionEntry}
   > n1 = GetDynamicMaterial(self=Get Video_Image())
   > SetTextureParameterValue(
    self=@n1
    ParameterName=MediaPlayer_param
    Value=GetMediaTexture!(self:Fort Movie Widget Object Reference) -> (ReturnValue:Media Texture Object Reference)
   )
   > SetStretch!(
    execute
    self:Scale Box Object Reference=Get Video_ScaleBox() -> (Video_ScaleBox:Scale Box Object Reference)
    InStretch:EStretch Enum=Select(Option 0:EStretch Enum=Fill, Option 1:EStretch Enum=ScaleToFit, Index=Get MaintainAspectRatio() -> (MaintainAspectRatio:bool)) -> (ReturnValue:EStretch Enum)
   ) -> (then)
  ]}
 StopPlaying {nodes [
   StopPlaying() {is=title K2Node_FunctionEntry}
   > RequestStopMovie(self:Fort Movie Widget Object Reference)
  ]}
]
variables [
 AutoPlay {type=boolean, default=true}
 MaintainAspectRatio {type=boolean}
 OnMediaFailedToOpen {}
 OnMediaReadyToPlay {}
 ShouldLoop {type=boolean, default=true}
]
widgets {
 info {inheritedWidgetCount=0, parentClass=/Script/FortniteUI.FortMovieWidget, rootWidgetClass=&Overlay, widgetCount=4}
 widgets [
  Overlay_0 {widgetClassPath=&Overlay, widget=Overlay_0}
  BackgroundImage {
   widgetClassPath=&Image
   parent=Overlay_0
   widget=BackgroundImage
   properties {
    visibility=HitTestInvisible
    brush {
     bIsDynamicallyLoaded=false
     drawAs=Image
     imageType=NoImage
     mirroring=NoMirror
     resourceName=None
     resourceObject=/Game/Global/Textures/Default/Blanks/T_BlackUserInterface
     tiling=NoTile
     imageSize=<16 16>
     margin=<ltrb 0 0 0 0>
     outlineSettings {
      bUseBrushTransparency=false
      roundingType=HalfHeightRadius
      width=0
      color {colorUseRule=UseColor_Specified, specifiedColor=<rgba 0 0 0 0>}
      cornerRadii=<0 0 0 0>
     }
     tintColor {colorUseRule=UseColor_Specified, specifiedColor=<rgba 1 1 1 1>}
     uVRegion {bIsValid=false, max=<0 0>, min=<0 0>}
    }
   }
   slot {class=OverlaySlot, properties {horizontalAlignment=HAlign_Fill, verticalAlignment=VAlign_Fill}}
  }
  Video_ScaleBox {
   widgetClassPath=&ScaleBox
   bIsVariable=true
   parent=Overlay_0
   widget=Video_ScaleBox
   slot {class=OverlaySlot, properties {horizontalAlignment=HAlign_Fill, verticalAlignment=VAlign_Fill}}
  }
  Video_Image {
   widgetClassPath=&Image
   bIsVariable=true
   parent=Video_ScaleBox
   widget=Video_Image
   properties {
    visibility=HitTestInvisible
    brush {
     bIsDynamicallyLoaded=false
     drawAs=Image
     imageType=NoImage
     mirroring=NoMirror
     resourceName=None
     resourceObject=/Game/Movies/Videos_Mat
     tiling=NoTile
     imageSize=<32 32>
     margin=<ltrb 0 0 0 0>
     outlineSettings {
      bUseBrushTransparency=false
      roundingType=HalfHeightRadius
      width=0
      color {colorUseRule=UseColor_Specified, specifiedColor=<rgba 0 0 0 0>}
      cornerRadii=<0 0 0 0>
     }
     tintColor {colorUseRule=UseColor_Specified, specifiedColor=<rgba 1 1 1 1>}
     uVRegion {bIsValid=false, max=<0 0>, min=<0 0>}
    }
   }
   slot {class=ScaleBoxSlot}
  }
 ]
}
