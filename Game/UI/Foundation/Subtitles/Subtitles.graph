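# HUD element widget that displays subtitles through a SubtitleDisplay root configured by the shared SubtitleOptions asset, with wrapped text.
# On initialization it starts an async listener for count changes of the Conversation.InteractingWithNPC gameplay tag on the owning player.
# While that count is above zero, a render translation shifts the subtitles upward; otherwise the offset resets.
# The widget list also includes a TutorialHighlight, which the graph does not touch.
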WidgetBlueprint /Game/UI/Foundation/Subtitles/Subtitles
graphs [EventGraph {nodes [
   Event OnInitialized()
   > n1 = WaitGameplayTagCountChangedOnActor!(execute:Exec, TargetActor:Actor Object Reference=GetOwningPlayer(), Tag:Gameplay Tag Structure=(TagName="Conversation.InteractingWithNPC")) -> (then:Exec, AsyncTaskProxy:Ability Async Wait Gameplay Tag Count Changed Object Reference, TagCountChanged:Exec, TagCount:int) {is=title K2Node_AsyncAction}
   n2 = Branch(execute=@n1.TagCountChanged, Condition="integer > integer"(A=@n1.TagCount))
   > SetRenderTranslation(Translation=(X=0,Y=-275))
   SetRenderTranslation(execute=@n2.else)
  ]}]
variables []
widgets {
 info {
  inheritedWidgetCount=1
  parentClass=&FortHUDElementWidget
  rootWidgetClass=/Script/SubtitlesWidgets.SubtitleDisplay
  widgetCount=2
 }
 widgets [
  SubtitleDisplay {
   widgetClassPath=&SubtitleDisplay
   bIsVariable=true
   widget=SubtitleDisplay
   properties {
    options=/Game/UI/SubtitleOptions
    previewText=SEE-Bot: SEE-Bot SEES THE INSIDE OF THE BAG. THERE IS SO MUCH BAG TO SEE.
    visibility=SelfHitTestInvisible
    wrapTextAt=1170
   }
  }
  TutorialHighlight {widgetClassPath=&TutorialHighlight}
 ]
}
