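# Cinematic UI state widget that plays full-screen movies with subtitles, putting the player controller in cinematic mode while active.
# A QueueCinematic client event creates a movie widget in MovieSlot, picks an audio track via a language-to-track table, ducks game audio and starts subtitles.
# Movie end fires CinematicFinishedPlaying and clears subtitles; SkipCinematic pauses and clears the movie.
# A 10-second delayed check ends the cinematic if playback failed.
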WidgetBlueprint /FortniteCoreUI/UI/Foundation/StateLayouts/Cinematic
graphs [
 ClearMovie {nodes [
   ClearMovie() {is=title K2Node_FunctionEntry}
   > ClearMovieWidget!(
    execute
    self
    MovieWidgetToClear:"Fort Movie Widget Object Reference (by ref)"=Get MovieWidget() -> (MovieWidget:Legacy Movie Widget Object Reference)
   ) -> (then)
   > ClearMovieSlot()
   > StopAndClearSubtitles()
  ]}
 ClearMovieSlot {nodes [
   ClearMovieSlot() {is=title K2Node_FunctionEntry}
   n1 = Get MovieSlot() -> (MovieSlot:Named Slot Object Reference)
   > ClearChildren(self=@n1)
   > SetVisibility(self=@n1, InVisibility=Collapsed)
  ]}
 ClearMovieWidget {nodes [
   n1 = ClearMovieWidget!() -> (then, MovieWidgetToClear:"Fort Movie Widget Object Reference (by ref)") {is=title K2Node_FunctionEntry}
   > Is Valid(InputObject=@n1)
   n2 = GetMediaPlayer!(self:Fort Movie Widget Object Reference=@n1) -> (ReturnValue:Media Player Object Reference)
   > n3 = Is Valid(InputObject=@n2)
   Set MovieWidget(execute=[@n3.Is Not Valid, @n4.then], MovieWidget:Legacy Movie Widget Object Reference) -> (Output_Get:Legacy Movie Widget Object Reference)
   n4 = Unbind Event from On End Reached(execute=@n3.Is Valid, self:Media Player Object Reference=@n2, Delegate=Create Event()) {is=title K2Node_RemoveDelegate}
  ]}
 ClearOldSubtitlesIfNecessary {nodes [
   n1 = ClearOldSubtitlesIfNecessary!() -> (then, NewSubtitlesExist:bool=~, NewSubtitles:Fort Media Subtitles Player Object Reference) {is=title K2Node_FunctionEntry}
   > n2 = Is Valid(InputObject=Get SubtitlesPlayer() -> (SubtitlesPlayer:Fort Media Subtitles Player Object Reference))
   > PrintString!(
    execute
    InString=Select(
     Option 0:String="Warning: Starting Empty new subtitles while old subtitles are still playing.  Stopping old subtitles since old audio should already be done or replaced."
     Option 1:String="Warning: Trying to start new subtitles while old subtitles are still playing.  Stopping old subtitles first."
     Index=@n1
    ) -> (ReturnValue:String)
    bPrintToScreen=false
    bPrintToLog
    TextColor
    Duration
    Key
   ) -> (then)
   > n3 = Stop(self:Fort Media Subtitles Player Object Reference=Get SubtitlesPlayer())
   Set SubtitlesPlayer(execute=[@n3.then, @n2.Is Not Valid], SubtitlesPlayer:Fort Media Subtitles Player Object Reference) -> (Output_Get:Fort Media Subtitles Player Object Reference)
  ]}
 EventGraph {nodes [
   HandleEndReached() {is=title K2Node_CustomEvent}
   > Set bFinished(bFinished:bool=true) -> (Output_Get:bool)
   > SetMoviePlaying!(execute, self, MoviePlaying:bool=false) -> (then)
   > SetVisibility(self=Get MovieSlot(), InVisibility=Collapsed)
   > Execute Client Event!(
    execute:Exec
    EventType:Fort Client Event Name Structure=(CategoryName="Cinematic",EventName="CinematicFinishedPlaying")
    EventSource:Object Reference
    EventFocus:Object Reference
   ) -> (then:Exec) {is=title K2Node_ExecuteClientEvent}
   > StopAndClearSubtitles()
   Event Construct()
   > SetVisibility(self=Get MovieSlot(), InVisibility=Collapsed)
   Event OnEnterState!() -> (OutputDelegate, then, PreviousUIState:EFortUIState Enum=Invalid) {is=title K2Node_Event}
   > SetCinematicMode!(
    execute
    self:Player Controller Object Reference=GetPlayerController!(PlayerIndex:int=0) -> (ReturnValue:Player Controller Object Reference)
    bInCinematicMode:bool=true
    bHidePlayer:bool=true
    bAffectsHUD:bool=false
    bAffectsMovement:bool=true
    bAffectsTurning:bool=true
   ) -> (then)
   Event Destruct()
   > ClearMovie() {is=title K2Node_CallFunction}
   > StopAndClearSubtitles()
   Event OnExitState!() -> (OutputDelegate, then, NextUIState:EFortUIState Enum=Invalid) {is=title K2Node_Event}
   > SetCinematicMode(
    self=GetPlayerController(PlayerIndex=0)
    bInCinematicMode=false
    bHidePlayer=true
    bAffectsHUD=true
    bAffectsMovement=true
    bAffectsTurning=true
   )
   n1 = Event PushContentWidgetInternal!() -> (OutputDelegate, then, Widget:Widget Object Reference, State:Content Push State Structure) {is=title K2Node_Event}
   > PushWidget!(
    execute
    self:Common Widget Stack Legacy Object Reference=Get MainContentStack() -> (MainContentStack:Common Widget Stack Legacy Object Reference)
    InWidget:Widget Object Reference=@n1.Widget
   ) -> (then)
   n2 = Handle Client Event: QueueCinematic!() -> (
    OutputDelegate:Delegate
    then:Exec
    EventSource:Object Reference
    EventFocus:Object Reference
    ClientEvent:"Fort Client Event Structure (by ref)"
   ) {is=title K2Node_HandleClientEvent, clientEventName {categoryName=Cinematic, eventName=QueueCinematic}}
   > n3 = Cast To FortMediaInfo!(execute, Object=@n2.EventSource) -> (then, CastFailed, AsFort Media Info:Fort Media Info Object Reference) {is=title K2Node_DynamicCast}
   > PlaySound2D()
   > Show Movie!(
    execute
    self
    MediaPlayer:Media Player Object Reference=Get MediaPlayer!(self:Fort Media Info Object Reference=@n3) -> (MediaPlayer:Media Player Object Reference)
    MediaSource:Media Source Object Reference=Get MediaSource!(self:Fort Media Info Object Reference=@n3) -> (MediaSource:Media Source Object Reference)
   ) -> (then)
   n4 = Cast To FortMediaSubtitlesPlayer!(Object=@n2.EventFocus) -> (AsFort Media Subtitles Player:Fort Media Subtitles Player Object Reference, bSuccess:bool=~) {is=title K2Node_DynamicCast}
   > ClearOldSubtitlesIfNecessary!(execute, self, NewSubtitlesExist:bool=@n4.bSuccess, NewSubtitles:Fort Media Subtitles Player Object Reference=@n4) -> (then) {is=title K2Node_CallFunction}
   > Branch(Condition=@n4.bSuccess)
   > BindToMediaPlayer!(
    execute
    self:Fort Media Subtitles Player Object Reference=@n4
    InMediaPlayer:Media Player Object Reference=GetMediaPlayer(self=Get MovieWidget())
   ) -> (then)
   > Play(self:Fort Media Subtitles Player Object Reference=@n4)
   > Set SubtitlesPlayer(SubtitlesPlayer=@n4)
   > UpdateWatermark(self:Fort Build Watermark Object Reference=Get Widget_BuildWatermark() -> (Widget_BuildWatermark:Build Watermark Object Reference))
   > Delay(Duration=10)
   > n5 = Branch(Condition=Get bFinished() -> (bFinished:bool))
   n6 = Is Cinematic Playing!(exec=@n5.else) -> (Cinematic Is Playing Or Preparing:Exec, CinematicIsNotPlaying:Exec) {is=title K2Node_MacroInstance}
   PrintString(execute=@n6.CinematicIsNotPlaying, InString=Failed to play movie. Ending cinematic., bPrintToScreen=false)
   > HandleEndReached() {is=title K2Node_CallFunction}
   Handle Client Event: SkipCinematic!() -> (
    OutputDelegate:Delegate
    then:Exec
    EventSource:Object Reference
    EventFocus:Object Reference
    ClientEvent:"Fort Client Event Structure (by ref)"
   ) {is=title K2Node_HandleClientEvent, clientEventName {categoryName=Cinematic, eventName=SkipCinematic}}
   > Set bFinished(bFinished=true)
   > SetMoviePlaying(MoviePlaying=false)
   > SkipMovie()
   Get CinematicMediaSoundWave() -> (CinematicMediaSoundWave:Object Reference)
  ]}
 GetAudioTrackIndex {nodes [
   GetAudioTrackIndex() {is=title K2Node_FunctionEntry}
   > n1 = GetDataTableRowNames(Table=/Game/UI/Foundation/StateLayouts/CinematicLanguageToTrackMap)
   > n2 = GetPreferredLanguages!(execute) -> (then, ReturnValue:Array of Strings)
   > n3 = For Each Loop(Array=@n2)
   Return Node!(execute=@n3.Completed, AudioTrack:int=0)
   > n4 = For Each Loop(Array:Array of Names=@n1) -> (Array Element:Name)
   > Branch(Condition="Equal, Case Insensitive (String)"(A="To String (Name)"(InName=@n4), B=@n3))
   > n5 = Get Data Table Row CinematicLanguageToTrackMap!(execute:Exec, DataTable:Data Table Object Reference=&CinematicLanguageToTrackMap, RowName:Name=@n4) -> (then:Exec, RowNotFound:Exec, ReturnValue:Cinematic Lanuage To Track Struct Structure) {is=title K2Node_GetDataTableRow}
   > Return Node(AudioTrack=Break Cinematic Lanuage to Track Struct!(CinematicLanuageToTrackStruct:"Cinematic Lanuage To Track Struct Structure (by ref)"=@n5) -> (TrackIndex_5_#1:int) {is=title K2Node_BreakStruct})
  ]}
 IsCinematicPlaying {nodes [
   Inputs!() -> (exec:Exec)
   n1 = Get MovieWidget()
   > n2 = Is Valid(InputObject=@n1)
   Outputs!(Cinematic Is Playing Or Preparing:Exec=@n4.then, CinematicIsNotPlaying:Exec=[@n4.else, @n2.Is Not Valid])
   n3 = GetMediaPlayer(self=@n1)
   > n4 = Branch(Condition=OR Boolean(
     A=IsPreparing!(self:Media Player Object Reference=@n3) -> (ReturnValue:bool)
     B=IsPlaying!(self:Media Player Object Reference=@n3) -> (ReturnValue:bool)
    ))
  ]}
 PopContentWidgetInternal {nodes [
   PopContentWidgetInternal!() -> (then, State:Content Push State Structure) {is=title K2Node_FunctionEntry}
   > n1 = PopWidget!(execute, self:Common Widget Stack Legacy Object Reference=Get MainContentStack()) -> (then, ReturnValue:Widget Object Reference)
   > Return Node!(execute, ReturnValue:Widget Object Reference=@n1)
  ]}
 SetMoviePlaying {nodes [
   n1 = SetMoviePlaying!() -> (then, MoviePlaying:bool=~) {is=title K2Node_FunctionEntry}
   > n2 = Branch(Condition=@n1)
   > PushSoundMixModifier!(execute, InSoundMixModifier:Sound Mix Object Reference=/Game/Sounds/FortSoundClassesAndModes/Fort_Cine_DuckIngame) -> (then)
   PopSoundMixModifier!(execute=@n2.else, InSoundMixModifier:Sound Mix Object Reference=&Fort_Cine_DuckIngame) -> (then)
  ]}
 Show Movie {nodes [
   n1 = Show Movie!() -> (then, MediaPlayer:Media Player Object Reference, MediaSource:Media Source Object Reference) {is=title K2Node_FunctionEntry}
   n2 = Get MovieSlot()
   > ClearChildren(self=@n2)
   > SetVisibility(self=@n2, InVisibility=SelfHitTestInvisible)
   > n3 = Create Widget!(
    execute:Exec
    Class:User Widget Class Reference=/Game/UI/Foundation/Movie/Legacy_MovieWidget.Legacy_MovieWidget_C
    OwningPlayer:Player Controller Object Reference=GetOwningPlayer()
    ShouldLoop:bool=false
    AutoPlay:bool=True
    MaintainAspectRatio:bool=true
   ) -> (then:Exec, ReturnValue:Legacy Movie Widget Object Reference)
   n4 = Get MovieWidget()
   > "Set Legacy Movie Widget Object Reference (by ref)"!(execute:Exec, Target:"Legacy Movie Widget Object Reference (by ref)"=@n4, Value:Legacy Movie Widget Object Reference=@n3) -> (then:Exec) {is=title K2Node_VariableSetRef}
   > n5 = GetAudioTrackIndex!(execute, self) -> (then, AudioTrack:int=0) {is=title K2Node_CallFunction}
   > Init!(
    execute
    self:Legacy Movie Widget Object Reference=@n4
    InMediaPlayer:Media Player Object Reference=@n1
    InMediaSource:Media Source Object Reference=@n1.MediaSource
    InMediaSoundWave:Object Reference
    AudioTrack:int=@n5
   ) -> (then)
   > Bind Event to On End Reached(self:Media Player Object Reference=GetMediaPlayer(self=Get MovieWidget()), Delegate=Create Event()) {is=title K2Node_AddDelegate}
   > SetMoviePlaying(MoviePlaying=true) {is=title K2Node_CallFunction}
   > AddChild(self=@n2, Content=@n4)
   Play!(execute, self:Legacy Movie Widget Object Reference, ShouldRewind:bool=false) -> (then)
  ]}
 SkipMovie {nodes [
   SkipMovie() {is=title K2Node_FunctionEntry}
   n1 = GetMediaPlayer(self=Get MovieWidget())
   > n2 = Branch(Condition=IsPlaying(self=@n1))
   > Pause!(execute, self:Media Player Object Reference=@n1) -> (then, ReturnValue:bool)
   > n3 = Execute Client Event(EventType=(CategoryName="Cinematic",EventName="CinematicFinishedPlaying"))
   ClearMovie(execute=[@n3.then, @n2.else])
  ]}
 StopAndClearSubtitles {nodes [
   StopAndClearSubtitles() {is=title K2Node_FunctionEntry}
   n1 = Get SubtitlesPlayer()
   > Is Valid(InputObject=@n1)
   > Stop(self=@n1)
   > Set SubtitlesPlayer()
  ]}
]
variables [
 bFinished {type=boolean}
 CinematicMediaSoundWave {type=Object, default=null}
 MovieWidget {type=Legacy_MovieWidget_C, default=null}
 SubtitlesPlayer {type=FortMediaSubtitlesPlayer, default=null}
]
widgets {
 info {inheritedWidgetCount=0, parentClass=/Script/FortniteUI.FortUIStateWidget_NUI, rootWidgetClass=&Overlay, widgetCount=7}
 widgets [
  OverlayMovieOverlay {widgetClassPath=&Overlay, bIsVariable=true, widget=OverlayMovieOverlay}
  LocalizedAudioOnlyMovieSlot {
   widgetClassPath=&NamedSlot
   bIsVariable=true
   parent=OverlayMovieOverlay
   widget=LocalizedAudioOnlyMovieSlot
   properties {bExposeOnInstanceOnly=true, toolTipText=This movie slot is to play localized audio only.}
   slot {class=OverlaySlot, properties {horizontalAlignment=HAlign_Fill, verticalAlignment=VAlign_Fill}}
  }
  MovieSlot {
   widgetClassPath=&NamedSlot
   bIsVariable=true
   parent=OverlayMovieOverlay
   widget=MovieSlot
   properties {bExposeOnInstanceOnly=true}
   slot {class=OverlaySlot, properties {horizontalAlignment=HAlign_Fill, verticalAlignment=VAlign_Fill}}
  }
  MainContentStack {
   widgetClassPath=/Script/CommonUILegacy.CommonWidgetStackLegacy
   bIsVariable=true
   parent=OverlayMovieOverlay
   widget=MainContentStack
   slot {class=OverlaySlot, properties {horizontalAlignment=HAlign_Fill, verticalAlignment=VAlign_Fill}}
  }
  TeamMicStack {
   widgetClassPath=/Game/UI/Foundation/MicIndicator/TeamMicStack
   bIsVariable=true
   parent=OverlayMovieOverlay
   widget=TeamMicStack
   slot {class=OverlaySlot}
  }
  Subtitles {
   widgetClassPath=/Game/UI/Foundation/Subtitles/Subtitles
   bIsVariable=true
   parent=OverlayMovieOverlay
   widget=Subtitles
   slot {class=OverlaySlot, properties {horizontalAlignment=HAlign_Center, verticalAlignment=VAlign_Bottom, padding=<ltrb 0 0 0 175>}}
  }
  Widget_BuildWatermark {
   widgetClassPath=/Game/UI/Foundation/Watermark/BuildWatermark
   bIsVariable=true
   parent=OverlayMovieOverlay
   widget=Widget_BuildWatermark
   slot {class=OverlaySlot, properties {horizontalAlignment=HAlign_Center, verticalAlignment=VAlign_Bottom, padding=<ltrb 0 0 0 4>}}
  }
 ]
}
