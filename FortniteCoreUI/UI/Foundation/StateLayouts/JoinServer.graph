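# JoinServer is a Fortnite UI state layout widget, built on FortUIStateWidget_JoinServer, for the join-server state.
# An overlay holds the inherited main content stack, filling the screen, with a centered build watermark above it.
# Entering the state updates the watermark.
# The tutorial dialog show event plays a confirm-skip-tutorial sound cue.
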WidgetBlueprint /FortniteCoreUI/UI/Foundation/StateLayouts/JoinServer
graphs [EventGraph {nodes [
   Event OnEnterState!() -> (OutputDelegate, then, PreviousUIState:EFortUIState Enum=Invalid) {is=title K2Node_Event}
   > UpdateWatermark(self:Fort Build Watermark Object Reference=Get BuildWatermark() -> (BuildWatermark:Build Watermark Object Reference))
   Event OnShowTutorialDialog() {is=title K2Node_Event}
   > PlaySound2D(Sound=/Game/Sounds/Fort_UI_Sounds/FrontEnd/Fort_Confirm_Skip_Tutorial_Cue)
  ]}]
variables []
widgets {
 info {inheritedWidgetCount=1, parentClass=/Script/FortniteUI.FortUIStateWidget_JoinServer, rootWidgetClass=&Overlay, widgetCount=3}
 widgets [
  Overlay_1 {widgetClassPath=&Overlay, widget=Overlay_1}
  Stack_MainContent {
   widgetClassPath=/Script/CommonUILegacy.CommonWidgetStackLegacy
   bInherited=true
   parent=Overlay_1
   widget=Stack_MainContent
   slot {class=OverlaySlot, properties {horizontalAlignment=HAlign_Fill, verticalAlignment=VAlign_Fill}}
  }
  BuildWatermark {
   widgetClassPath=/Game/UI/Foundation/Watermark/BuildWatermark
   bIsVariable=true
   parent=Overlay_1
   widget=BuildWatermark
   slot {class=OverlaySlot, properties {horizontalAlignment=HAlign_Center, padding=<ltrb 0 64 0 0>}}
  }
 ]
}
